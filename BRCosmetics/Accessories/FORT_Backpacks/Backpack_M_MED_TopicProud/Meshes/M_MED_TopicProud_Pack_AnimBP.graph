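# Animation Blueprint for the TopicProud back bling that adds secondary motion to its bones and spins MSTR_cap while skydiving, jumping or in the front end.
# The event graph maps owner spine yaw to a pack offset and rotation, and drives cap spin from real time, with interpolated jump and front-end weights.
# AnimDynamics chains jiggle compass, bananas, hands, chiken and candy, using cached poses to switch between front-end and skydive/parachute tunings.
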AnimBlueprint /BRCosmetics/Accessories/FORT_Backpacks/Backpack_M_MED_TopicProud/Meshes/M_MED_TopicProud_Pack_AnimBP
target_skeleton=&/M_MED_TopicProud_Pack_Skeleton
graphs [
 AnimGraph {nodes [
   n1 = Anim Dynamics - Bone: chiken(ComponentPose=Local To Component(LocalPose=Use cached pose 'forskyd'()), Alpha=1, ExternalForce="0,0,0") {
    angularDampingOverride=1
    bOverrideLinearDamping=true
    bUseGravityOverride=true
    linearDampingOverride=1
    simulationSpace=BoneRelative
    boundBone {boneName=chiken}
    componentAppliedLinearAccClamp=<4000 4000 4000>
    componentLinearAccScale=<0.2 0.2 0.3>
    componentLinearVelScale=<0.2 0.2 0.16>
    gravityOverride {z=-980}
    physicsBodyDefinitions [{
      boundBone {boneName=chiken}
      boxExtents=<8 8 8>
      constraintSetup {
       angularTargetAxis=AxisY
       angularLimitsMax=<0 0 0>
       angularLimitsMin=<0 0 0>
       angularTarget=<1 0 0>
       linearAxesMax=<0.4 0 10>
       linearAxesMin=<-0.1 0 0>
      }
      localJointOffset=<0 0 0.6>
     }]
    relativeSpaceBone {boneName=root}
    simSpaceSettings {maxAngularAcceleration=15, maxAngularVelocity=5}
   }
   n2 = Anim Dynamics - Bone: MSTR_candy(ComponentPose=@n1, Alpha=1, ExternalForce="0,0,0") {
    angularDampingOverride=1
    bOverrideLinearDamping=true
    bUseGravityOverride=true
    linearDampingOverride=1
    simulationSpace=BoneRelative
    boundBone {boneName=MSTR_candy}
    componentAppliedLinearAccClamp=<4000 4000 4000>
    componentLinearAccScale=<0.2 0.2 0.34>
    componentLinearVelScale=<0.2 0.2 0.2>
    gravityOverride {z=-980}
    physicsBodyDefinitions [{
      boundBone {boneName=MSTR_candy}
      boxExtents=<10 10 10>
      constraintSetup {angularTargetAxis=AxisY, angularLimitsMax=<0 0 0>, angularLimitsMin=<0 0 0>, angularTarget=<1 0 0>, linearAxesMax=<10 0 0>}
      localJointOffset=<0.7 0 0>
     }]
    relativeSpaceBone {boneName=root}
    simSpaceSettings {maxAngularAcceleration=15, maxAngularVelocity=5}
   }
   n3 = Anim Dynamics - Bone: chiken(ComponentPose=Local To Component(LocalPose=Use cached pose 'forskyd'()), Alpha=1, ExternalForce="0,0,0") {
    angularDampingOverride=1
    bOverrideLinearDamping=true
    bUseGravityOverride=true
    linearDampingOverride=1
    simulationSpace=BoneRelative
    boundBone {boneName=chiken}
    componentAppliedLinearAccClamp=<4000 4000 4000>
    componentLinearAccScale=<0.2 0.2 0.1>
    componentLinearVelScale=<0.2 0.2 0.05>
    gravityOverride {z=-980}
    physicsBodyDefinitions [{
      boundBone {boneName=chiken}
      boxExtents=<8 8 8>
      constraintSetup {angularTargetAxis=AxisY, angularLimitsMax=<0 0 0>, angularLimitsMin=<0 0 0>, angularTarget=<1 0 0>, linearAxesMax=<0 0 1>}
      localJointOffset=<0 0 0.4>
     }]
    relativeSpaceBone {boneName=root}
    simSpaceSettings {maxAngularAcceleration=15, maxAngularVelocity=5}
   }
   n4 = Anim Dynamics - Bone: MSTR_candy(ComponentPose=@n3, Alpha=1, ExternalForce="0,0,0") {
    angularDampingOverride=1
    bOverrideLinearDamping=true
    bUseGravityOverride=true
    linearDampingOverride=1
    simulationSpace=BoneRelative
    boundBone {boneName=MSTR_candy}
    componentAppliedLinearAccClamp=<4000 4000 4000>
    componentLinearAccScale=<0.2 0.2 0.1>
    componentLinearVelScale=<0.2 0.2 0.05>
    gravityOverride {z=-980}
    physicsBodyDefinitions [{
      boundBone {boneName=MSTR_candy}
      boxExtents=<10 10 10>
      constraintSetup {angularTargetAxis=AxisY, angularLimitsMax=<0 0 0>, angularLimitsMin=<0 0 0>, angularTarget=<1 0 0>, linearAxesMax=<1 0 0>}
      localJointOffset=<0.4 0 0>
     }]
    relativeSpaceBone {boneName=root}
    simSpaceSettings {maxAngularAcceleration=15, maxAngularVelocity=5}
   }
   Output Pose(Result=Blend Poses by bool(
     BlendPose_0=Component To Local(ComponentPose=@n4)
     BlendPose_1=Component To Local(ComponentPose=@n2)
     bActiveValue=Get isfullskydivie() -> (isfullskydivie:bool)
    )) {layerGroup=None}
   n5 = "Transform (Modify) Bone - Bone: MATR_Hand_l"(ComponentPose=Local To Component(LocalPose=Use cached pose 'FE DYN'()), Alpha=1, Translation="0,0,0", Rotation="2,-20,-5", Scale="1,1,1") {rotationMode=BMM_Additive, rotationSpace=BCS_BoneSpace, boneToModify {boneName=MATR_Hand_l}, rotation=<rot 2 -20 -5>}
   n6 = Anim Dynamics - Bone: Hand01_l(ComponentPose=@n5, Alpha=1, ExternalForce="0,0,0") {
    angularDampingOverride=1
    angularSpringConstant=800
    bAngularSpring=true
    bOverrideAngularDamping=true
    bUseGravityOverride=true
    simulationSpace=BoneRelative
    boundBone {boneName=Hand01_l}
    componentAppliedLinearAccClamp=<4000 4000 4000>
    componentLinearAccScale=<0.3 0.3 0.2>
    componentLinearVelScale=<0.3 0.3 0.13>
    gravityOverride {z=-980}
    physicsBodyDefinitions [{
      boundBone {boneName=Hand01_l}
      boxExtents=<10 10 10>
      constraintSetup {angularLimitsMax=<0 2 25>, angularLimitsMin=<0 0 -30>, angularTarget=<1 0 0>}
      localJointOffset=<1 0 0>
     }]
    relativeSpaceBone {boneName=root}
   }
   n7 = Anim Dynamics - Bone: Hand02_l(ComponentPose=@n6, Alpha=1, ExternalForce="0,0,0") {
    angularDampingOverride=1
    angularSpringConstant=500
    bOverrideAngularDamping=true
    bUseGravityOverride=true
    simulationSpace=BoneRelative
    boundBone {boneName=Hand02_l}
    componentAppliedLinearAccClamp=<4000 4000 4000>
    componentLinearAccScale {z=0.2}
    componentLinearVelScale {z=0.2}
    gravityOverride {z=-980}
    physicsBodyDefinitions [{
      boundBone {boneName=Hand02_l}
      boxExtents=<11 11 11>
      constraintSetup {angularLimitsMax=<0 2 25>, angularLimitsMin=<0 1 -25>, angularTarget=<1 0 0>}
      localJointOffset=<1 0 0>
     }]
    relativeSpaceBone {boneName=root}
    simSpaceSettings {maxAngularAcceleration=15, maxAngularVelocity=4}
   }
   n8 = Anim Dynamics - Bone: chiken(ComponentPose=@n7, Alpha=1, ExternalForce="0,0,0") {
    angularDampingOverride=0.9999
    angularSpringConstant=1000
    bAngularSpring=true
    bOverrideAngularDamping=true
    bUseGravityOverride=true
    simulationSpace=BoneRelative
    boundBone {boneName=chiken}
    componentAppliedLinearAccClamp=<4000 4000 4000>
    componentLinearAccScale=<0.5 0.5 0.5>
    componentLinearVelScale=<0.3 0.3 0.2>
    gravityOverride {z=-980}
    physicsBodyDefinitions [{
      boundBone {boneName=chiken}
      boxExtents=<20 20 20>
      constraintSetup {angularTargetAxis=AxisZ, angularLimitsMax=<15 2 4>, angularLimitsMin=<-4 0 -4>, angularTarget=<0 0 1>}
      localJointOffset=<0 0 1>
     }]
    relativeSpaceBone {boneName=root}
    simSpaceSettings {maxAngularAcceleration=15, maxAngularVelocity=4}
   }
   n9 = Anim Dynamics - Bone: MSTR_candy(ComponentPose=@n8, Alpha=1, ExternalForce="0,0,0") {
    angularDampingOverride=1
    bOverrideAngularDamping=true
    bUseGravityOverride=true
    simulationSpace=BoneRelative
    boundBone {boneName=MSTR_candy}
    componentAppliedLinearAccClamp=<4000 4000 4000>
    componentLinearAccScale=<0.2 0.2 0.32>
    componentLinearVelScale=<0.2 0.2 0.22>
    gravityOverride {z=-980}
    physicsBodyDefinitions [{
      boundBone {boneName=MSTR_candy}
      boxExtents=<10 10 10>
      constraintSetup {angularTargetAxis=AxisY, angularLimitsMax=<0 13 10>, angularLimitsMin=<0 1 0>, angularTarget=<1 0 0>}
      localJointOffset=<1 0 0>
     }]
    relativeSpaceBone {boneName=root}
    simSpaceSettings {maxAngularAcceleration=15, maxAngularVelocity=5}
   }
   n10 = "Transform (Modify) Bone - Bone: MATR_Hand_l"(
    ComponentPose=Local To Component(LocalPose=Use cached pose 'FE DYN'())
    Alpha=1
    Translation="0,0,0"
    Rotation="-4.252,-19.149,-5.651"
    Scale="1,1,1"
   ) {rotationMode=BMM_Additive, rotationSpace=BCS_BoneSpace, boneToModify {boneName=MATR_Hand_l}, rotation=<rot -4.252 -19.149 -5.651>}
   n11 = Anim Dynamics - Bone: Hand01_l(ComponentPose=@n10, Alpha=1, ExternalForce="0,0,0") {
    angularDampingOverride=1
    angularSpringConstant=800
    bAngularSpring=true
    bOverrideAngularDamping=true
    bUseGravityOverride=true
    simulationSpace=BoneRelative
    boundBone {boneName=Hand01_l}
    componentAppliedLinearAccClamp=<4000 4000 4000>
    componentLinearAccScale=<0.3 0.3 0.2>
    componentLinearVelScale=<0.3 0.3 0.13>
    gravityOverride {z=-980}
    physicsBodyDefinitions [{
      boundBone {boneName=Hand01_l}
      boxExtents=<10 10 10>
      constraintSetup {angularLimitsMax=<0 2 25>, angularLimitsMin=<0 0 -30>, angularTarget=<1 0 0>}
      localJointOffset=<1 0 0>
     }]
    relativeSpaceBone {boneName=root}
   }
   n12 = Anim Dynamics - Bone: Hand02_l(ComponentPose=@n11, Alpha=1, ExternalForce="0,0,0") {
    angularDampingOverride=1
    angularSpringConstant=500
    bOverrideAngularDamping=true
    bUseGravityOverride=true
    simulationSpace=BoneRelative
    boundBone {boneName=Hand02_l}
    componentAppliedLinearAccClamp=<4000 4000 4000>
    componentLinearAccScale {z=0.2}
    componentLinearVelScale {z=0.2}
    gravityOverride {z=-980}
    physicsBodyDefinitions [{
      boundBone {boneName=Hand02_l}
      boxExtents=<11 11 11>
      constraintSetup {angularLimitsMax=<0 2 25>, angularLimitsMin=<0 1 -25>, angularTarget=<1 0 0>}
      localJointOffset=<1 0 0>
     }]
    relativeSpaceBone {boneName=root}
    simSpaceSettings {maxAngularAcceleration=15, maxAngularVelocity=4}
   }
   n13 = Anim Dynamics - Bone: chiken(ComponentPose=@n12, Alpha=1, ExternalForce="0,0,0") {
    angularDampingOverride=0.9999
    angularSpringConstant=1000
    bAngularSpring=true
    bOverrideAngularDamping=true
    bUseGravityOverride=true
    simulationSpace=BoneRelative
    boundBone {boneName=chiken}
    componentAppliedLinearAccClamp=<4000 4000 4000>
    componentLinearAccScale=<0.62 0.62 0.5>
    componentLinearVelScale=<0.3 0.3 0.2>
    gravityOverride {z=-980}
    physicsBodyDefinitions [{
      boundBone {boneName=chiken}
      boxExtents=<20 20 20>
      constraintSetup {angularTargetAxis=AxisZ, angularLimitsMax=<15 3 4>, angularLimitsMin=<-4 -2 -4>, angularTarget=<0 0 1>}
      localJointOffset=<0 0 1>
     }]
    relativeSpaceBone {boneName=root}
    simSpaceSettings {maxAngularAcceleration=15, maxAngularVelocity=4}
   }
   n14 = Anim Dynamics - Bone: MSTR_candy(ComponentPose=@n13, Alpha=1, ExternalForce="0,0,0") {
    angularDampingOverride=1
    bOverrideAngularDamping=true
    bUseGravityOverride=true
    simulationSpace=BoneRelative
    boundBone {boneName=MSTR_candy}
    componentAppliedLinearAccClamp=<4000 4000 4000>
    componentLinearAccScale=<0.3 0.3 0.32>
    componentLinearVelScale=<0.2 0.2 0.22>
    gravityOverride {z=-980}
    physicsBodyDefinitions [{
      boundBone {boneName=MSTR_candy}
      boxExtents=<10 10 10>
      constraintSetup {angularTargetAxis=AxisY, angularLimitsMax=<0 9 8>, angularLimitsMin=<0 1 2>, angularTarget=<1 0 0>}
      localJointOffset=<1 0 0>
     }]
    relativeSpaceBone {boneName=root}
    simSpaceSettings {maxAngularAcceleration=15, maxAngularVelocity=5}
   }
   Save cached pose 'forskyd'(Pose=Blend Poses by bool(BlendPose_0=Component To Local(ComponentPose=@n14), BlendPose_1=Component To Local(ComponentPose=@n9), bActiveValue=Get bIsInFrontEnd())) {cacheName=forskyd}
   n15 = Local To Component(LocalPose=Input Pose())
   n16 = "Transform (Modify) Bone - Bone: pack"(
    ComponentPose="Transform (Modify) Bone - Bone: pack"(
     ComponentPose="Transform (Modify) Bone - Bone: chiken"(ComponentPose=@n15, Alpha=1, Translation="0,0,0.35", Rotation="0,0,0", Scale="1,1,1") {translationMode=BMM_Additive, translationSpace=BCS_BoneSpace, boneToModify {boneName=chiken}, translation {z=0.35}}
     Translation="0,4,0"
     Rotation="0,0,0"
    ) {translationMode=BMM_Additive, translationSpace=BCS_BoneSpace, boneToModify {boneName=pack}, translation {y=4}}
    Translation=Get backpack_translation()
    Rotation=Get backpack_rotate()
   ) {rotationMode=BMM_Additive, translationMode=BMM_Additive, alphaBoolBlend {blendOption=Linear}, boneToModify {boneName=pack}}
   n17 = "Transform (Modify) Bone - Bone: MSTR_bananas"(ComponentPose=@n16, Alpha=1, Translation="-1,0,0", Rotation="0,0,0", Scale="1,1,1") {translationMode=BMM_Additive, translationSpace=BCS_BoneSpace, boneToModify {boneName=MSTR_bananas}, translation {x=-1}}
   n18 = "Transform (Modify) Bone - Bone: MSTR_cap"!(
    ComponentPose=@n17
    bAlphaBoolEnabled:bool=Get move porp_s() -> (move porp_s:bool)
    Translation="0,0,0"
    Rotation=Get prop_move sky() -> (prop_move sky:Rotator)
    Scale="1,1,1"
   ) -> (Pose) {alphaInputType=Bool, rotationMode=BMM_Additive, rotationSpace=BCS_BoneSpace, boneToModify {boneName=MSTR_cap}}
   n19 = Get My() -> (My:double)
   n20 = Get prop_move jump() -> (prop_move jump:Rotator)
   n21 = Anim Dynamics - Bone: compass(
    ComponentPose="Transform (Modify) Bone - Bone: MSTR_cap"!(
     ComponentPose="Transform (Modify) Bone - Bone: MSTR_cap"(ComponentPose=@n18, Alpha=@n19, Translation="0,0,0", Rotation=@n20, Scale="1,1,1") {rotationMode=BMM_Additive, rotationSpace=BCS_BoneSpace, boneToModify {boneName=MSTR_cap}}
     Alpha:float=Get Me() -> (Me:double)
     Translation="0,0,0"
     Rotation=Get prop_move FE() -> (prop_move FE:Rotator)
     Scale="1,1,1"
    ) -> (Pose) {rotationMode=BMM_Additive, rotationSpace=BCS_BoneSpace, boneToModify {boneName=MSTR_cap}}
    Alpha=1
    ExternalForce="0,0,0"
   ) {
    angularDampingOverride=1
    angularSpringConstant=1000
    bAngularSpring=true
    bLinearSpring=true
    bOverrideAngularDamping=true
    bUseGravityOverride=true
    linearSpringConstant=50
    simulationSpace=BoneRelative
    boundBone {boneName=compass}
    componentLinearAccScale {z=0.24}
    componentLinearVelScale {z=0.05}
    gravityOverride {z=-700}
    physicsBodyDefinitions [{
      boundBone {boneName=compass}
      boxExtents=<20 20 20>
      constraintSetup {angularLimitsMax=<0 2 5>, angularLimitsMin=<0 -5 -5>, angularTarget=<1 0 0>, linearAxesMax=<0 0 5>}
      localJointOffset=<1 0 0>
     }]
    relativeSpaceBone {boneName=root}
   }
   n22 = Anim Dynamics - Bone: bananas(ComponentPose=@n21, Alpha=1, ExternalForce="0,0,0") {
    angularDampingOverride=1
    angularSpringConstant=1000
    bAngularSpring=true
    bLinearSpring=true
    bOverrideAngularDamping=true
    bUseGravityOverride=true
    linearSpringConstant=200
    simulationSpace=BoneRelative
    boundBone {boneName=bananas}
    componentAppliedLinearAccClamp=<4000 4000 4000>
    componentLinearAccScale=<0.3 0.3 0.3>
    componentLinearVelScale=<0.3 0.3 0.3>
    gravityOverride {z=-980}
    physicsBodyDefinitions [{
      boundBone {boneName=bananas}
      boxExtents=<10 10 10>
      constraintSetup {angularLimitsMax=<0 0 5>, angularLimitsMin=<0 -8 -5>, angularTarget=<1 0 0>, linearAxesMax=<0 0 5>}
      localJointOffset=<2 0 0>
     }]
    relativeSpaceBone {boneName=root}
   }
   n23 = Anim Dynamics - Bone: Hand01_r(ComponentPose=@n22, Alpha=1, ExternalForce="0,0,0") {
    angularDampingOverride=0.999
    angularSpringConstant=1500
    bAngularSpring=true
    bOverrideAngularDamping=true
    bUseGravityOverride=true
    simulationSpace=BoneRelative
    boundBone {boneName=Hand01_r}
    componentAppliedLinearAccClamp=<4000 4000 4000>
    componentLinearAccScale=<0.7 0.95 0.4>
    componentLinearVelScale=<0.3 0.3 0.3>
    gravityOverride {z=-980}
    physicsBodyDefinitions [{
      boundBone {boneName=Hand01_r}
      boxExtents=<20 20 20>
      constraintSetup {angularLimitsMax=<0 15 5>, angularLimitsMin=<0 -50 -5>, angularTarget=<1 0 0>}
      localJointOffset=<2 0 0>
     }]
    relativeSpaceBone {boneName=root}
    simSpaceSettings {maxAngularAcceleration=15, maxAngularVelocity=4}
   }
   n24 = Anim Dynamics - Bone: Hand02_r(ComponentPose=@n23, Alpha=1, ExternalForce="0,0,0") {
    angularDampingOverride=0.999
    angularSpringConstant=1500
    bAngularSpring=true
    bOverrideAngularDamping=true
    bUseGravityOverride=true
    simulationSpace=BoneRelative
    boundBone {boneName=Hand02_r}
    componentAppliedLinearAccClamp=<4000 4000 4000>
    componentLinearAccScale=<0.7 0.95 0.35>
    componentLinearVelScale=<0.3 0.3 0.25>
    gravityOverride {z=-980}
    physicsBodyDefinitions [{
      boundBone {boneName=Hand02_r}
      boxExtents=<20 20 20>
      constraintSetup {angularLimitsMax=<0 5 30>, angularLimitsMin=<0 -45 -30>, angularTarget=<1 0 0>}
      localJointOffset=<3 0 0>
     }]
    relativeSpaceBone {boneName=root}
    simSpaceSettings {maxAngularAcceleration=15, maxAngularVelocity=4}
   }
   n25 = Anim Dynamics - Bone: MSTR_front_stop(ComponentPose=@n24, Alpha=1, ExternalForce="0,0,0") {
    angularDampingOverride=1
    angularSpringConstant=1000
    bAngularSpring=true
    bOverrideAngularDamping=true
    bUseGravityOverride=true
    simulationSpace=BoneRelative
    boundBone {boneName=MSTR_front_stop}
    componentAppliedLinearAccClamp=<4000 4000 4000>
    componentLinearAccScale=<0.3 0.3 0.3>
    componentLinearVelScale=<0.3 0.3 0.3>
    gravityOverride {z=-980}
    physicsBodyDefinitions [{
      boundBone {boneName=MSTR_front_stop}
      boxExtents=<10 10 10>
      constraintSetup {angularLimitsMax=<0 5 15>, angularLimitsMin=<0 -15 -15>, angularTarget=<1 0 0>}
      localJointOffset=<5 0 0>
     }]
    relativeSpaceBone {boneName=root}
   }
   Save cached pose 'FE DYN'(Pose=Component To Local(ComponentPose=@n25)) {cacheName=FE DYN}
  ]}
 EventGraph {nodes [
   n1 = Event BlueprintUpdateAnimation()
   > Set Delta Time X(Delta Time X=@n1.DeltaTimeX)
   > n2 = Sequence!(execute) -> (then_0, then_1, then_2:Exec, then_3:Exec, then_4:Exec)
   n3 = float * float(A=GetRealTimeSeconds(), B=80) {is=title K2Node_PromotableOperator}
   Set prop_move FE(execute=@n2.then_3, prop_move FE:Rotator=MakeRotator(Roll=float * float(A=@n3, B=7))) -> (Output_Get:Rotator)
   > n4 = Branch(Condition=NOT Boolean(A=Get bIsInFrontEnd()))
   > Set Me(Me:double=FInterpTo(Current=Get Me(), Target=0, DeltaTime=Get Delta Time X(), InterpSpeed=5)) -> (Output_Get:double)
   Set Me(execute=@n4.else, Me=FInterpTo(Current=Get Me(), Target=1, DeltaTime=Get Delta Time X(), InterpSpeed=7))
   n5 = Set move porp_s(execute=@n2.then_1, move porp_s:bool=Get bIsSkyDiving()) -> (Output_Get:bool)
   > Branch(Condition=@n5)
   n6 = float * float(A=GetRealTimeSeconds(), B=110)
   > Set prop_move sky(prop_move sky:Rotator=MakeRotator(Roll=float * float(A=@n6, B=7))) -> (Output_Get:Rotator)
   Set isfullskydivie(
    execute=@n2.then_4
    isfullskydivie:bool=OR Boolean!(A=Get bIsSkyDiving(), B=Get bIsSkydiveDiveMode(), C:bool=Get bIsSkydiveFloating(), D:bool=Get bIsParachuteOpened()) -> (ReturnValue)
   ) -> (Output_Get:bool)
   n7 = GetOwnerSkeletalMeshComponent()
   n8 = BreakRotator(InRot=BreakTransform(InTransform=GetDeltaTransformFromRefPose(self=@n7, BoneName=spine_05, BaseName=spine_01)).Rotation)
   n9 = MapRangeClamped(Value=@n8.Yaw, InRangeA=-10, InRangeB=-50, OutRangeB=55)
   Set backpack_rotate(execute=@n2.then_0, backpack_rotate=MakeRotator(Roll=@n9, Pitch=float / float(A=@n9, B=10)))
   > Set backpack_translation(backpack_translation=MakeVector(Y=MapRangeClamped(Value=@n8.Yaw, InRangeA=-15, InRangeB=-55, OutRangeB=-11), Z=-6))
   n10 = float * float(A=GetRealTimeSeconds(), B=120)
   Set prop_move jump(execute=@n2.then_2, prop_move jump:Rotator=MakeRotator(Roll=float * float(A=@n10, B=8))) -> (Output_Get:Rotator)
   > n11 = Branch(Condition=NOT Boolean(A=Get bIsJumping()))
   > Set My(My:double=FInterpTo(Current=Get My(), Target=0, DeltaTime=Get Delta Time X(), InterpSpeed=5)) -> (Output_Get:double)
   Set My(execute=@n11.else, My=FInterpTo(Current=Get My(), Target=1, DeltaTime=Get Delta Time X(), InterpSpeed=7))
   TryGetPawnOwner()
  ]}
]
variables [
 backpack_rotate {type=Rotator}
 backpack_translation {type=Vector}
 Delta Time X {type=number}
 isfullskydivie {type=boolean}
 Me {type=number}
 move porp_s {type=boolean}
 move prop_j {type=boolean}
 My {type=number}
 prop_move FE {type=Rotator}
 prop_move jump {type=Rotator}
 prop_move sky {type=Rotator}
]
